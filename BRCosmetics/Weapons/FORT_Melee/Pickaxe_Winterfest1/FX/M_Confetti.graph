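# Masked, unlit, two-sided Niagara sprite material for animated, multicolored confetti in the Winterfest pickaxe effects.
# Time plus per-particle random drives flipbook phase for a confetti normal map and an alpha mask multiplied by particle alpha.
# A pulsing red base is hue-shifted per particle, tinted by particle color and modulated by scrolling sine shimmer.
# Flipbook normals distort the shimmer and fake shading via a dot product, feeding an emissive adjustment function.
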Material /BRCosmetics/Weapons/FORT_Melee/Pickaxe_Winterfest1/FX/M_Confetti
graph {
 properties {
  bHasPixelAnimation=true
  bUsedWithNiagaraSprites=true
  blendMode=BLEND_Masked
  opacityMaskClipValue=0.1
  pixelDepthOffsetMode=PDOM_Legacy
  shadingModel=MSM_Unlit
  twoSided=true
  usedShadingModels=Unlit
 }
 nodes [
  n1 = TextureCoordinate()
  n2 = Add(A=ParticleRandom(), B=Time())
  n3 = Multiply(A=ComponentMask(None=Add(A=@n1, B=@n2)) {g=true}) {constB=0.25}
  n4 = ParticleRandom()
  n5 = Constant() {r=2}
  n6 = Multiply(A=MaterialFunctionCall(Speed=Multiply(A=@n4, B=@n5)) {materialFunction=&DebugTimeSine}.Result)
  n7 = MaterialFunctionCall(
   Hue Shift Percentage=Multiply(A=@n4) {constB=0.8}
   Texture=Multiply(A=Clamp(None=@n6) {maxDefault=5, minDefault=0.25}, B=Constant3Vector() {constant=<rgba 1 0 0 0>})
  ) {materialFunction=&HueShift}
  n8 = ParticleColor()
  n9 = Add(A=@n8.RGB) {constB=0.1}
  n10 = Multiply(A=ConstantBiasScale(None=Sine(None=@n3)), B=Multiply(A=@n7.Result, B=Saturate(None=@n9)))
  n11 = Constant() {r=1}
  n12 = NamedRerouteUsage()
  n13 = TextureObjectParameter() {parameterName=Confetti_Normal0, samplerType=SAMPLERTYPE_Normal, texture=/Game/Effects/Fort_Effects/Textures/Confetti/T_Confetti_SubUV_4X4_03}
  n14 = MaterialFunctionCall("Animation  Phase (0-1)"=NamedRerouteUsage(), Number of Rows=@n12, Number of Columns=@n12, Texture=@n13) {materialFunction=&FlipBook}
  n15 = DotProduct(A=@n14.Result, B=Constant3Vector() {constant=<rgba 0 0 1 0>})
  n16 = ComponentMask(None=LinearInterpolate(A=TextureCoordinate(), B=ComponentMask(None=@n14.Result) {g=true, r=true}) {constAlpha=0.2}) {g=true}
  n17 = Add(A=Time(), B=NamedRerouteUsage())
  n18 = Sine(None=Multiply(A=Add(A=@n16, B=@n17)) {constB=1.5})
  n19 = Multiply(A=@n7.Result, B=Add(A=Power(Base=@n15), B=ConstantBiasScale(None=@n18)))
  n20 = Multiply(A=@n8.RGB, B=@n19)
  n21 = Constant() {desc="Shadowed (non light transmission) intensity", r=0.75}
  MaterialFunctionCall(Emissive=Add(A=Multiply(A=@n10, B=@n11) {constB=2}, B=Multiply(A=@n20, B=@n21) {constB=0.15})) {materialFunction=&MF_FX_EmissiveAdjustments}
  Multiply(A=@n19, B=MaterialFunctionCall() {materialFunction=/Game/Effects/Fort_Effects/Materials/_Functions/CurveColor_Time_of_Day_S23}.Result)
  n22 = Time()
  n23 = NamedRerouteUsage()
  NamedRerouteDeclaration(None=Frac(None=Add(A=@n22, B=@n23))) {name=Animation Phase, nodeColor=<rgba 0 1 0.2031 1>}
  NamedRerouteDeclaration(None=Constant() {r=6}) {name=RowsAndColumns, nodeColor=<rgba 0 1 0.4375 1>}
  NamedRerouteDeclaration(None=ParticleRandom()) {name=Random, nodeColor=<rgba 0 0.7422 1 1>}
  n24 = NamedRerouteUsage()
  n25 = NamedRerouteUsage()
  n26 = TextureObject() {texture=/Game/Effects/Fort_Effects/Textures/Confetti/T_Confetti_SubUV_4X4_A}
  n27 = Multiply(
   A=ParticleColor().A
   B=ComponentMask(None=MaterialFunctionCall("Animation  Phase (0-1)"=@n24, Number of Rows=@n25, Number of Columns=@n25, Texture=@n26) {materialFunction=&FlipBook}.Result) {b=true}
  )
  NamedRerouteDeclaration(None=@n27) {name=Alpha, nodeColor=<rgba 0.4062 0 1 1>}
  Reroute(None=@n27)
  Saturate(None=Add(A=@n8.RGB) {constB=0.5})
 ]
}
