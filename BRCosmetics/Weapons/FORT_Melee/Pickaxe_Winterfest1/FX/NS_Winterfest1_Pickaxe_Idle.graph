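# Idle Niagara effect for the Winterfest1 pickaxe: snow-globe-like swirl of snow and confetti, plus a glow sprite, tied to the skeletal mesh.
# GPU and smaller CPU emitters burst looping particles, held near a sampled mesh point by non-rigid pendulum constraints.
# Owner velocity drives acceleration, curl noise and drag, with gradual settling; User.bIsFrontEnd instead uses gentle noise without downward force.
# Time of day tints snow and glow.
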NiagaraSystem /BRCosmetics/Weapons/FORT_Melee/Pickaxe_Winterfest1/FX/NS_Winterfest1_Pickaxe_Idle
dependencies {
 usedDataInterfaces [
  &NiagaraDataInterfaceSpriteRendererInfo
  &NiagaraDataInterfaceCurve
  &NiagaraDataInterfaceSkeletalMesh
 ]
 usedDynamicInputs []
 usedModules [
  &self:SystemSpawnScript.SystemScriptSource.SystemScriptGraph.NiagaraNodeAssignment_0.SetVariables_#1
  &SystemState
 ]
}
emitters {
 Confetti {
  simTarget=GPUComputeSim
  emitterUpdateScript [
   &EmitterState(
    Inactive Response=::NewEnumerator1
    Life Cycle Mode=::NewEnumerator1
    Loop Behavior=::NewEnumerator1
    Loop Duration Mode=::NewEnumerator1
    UseLoopDelay=false
   )
   &SpawnBurst_Instantaneous(Spawn Count=75, Use Spawn Probability=false, UseLoopCountLimit=false)
  ]
  particleSpawnScript [
   &InitializeParticle(
    Color Mode=::NewEnumerator0
    Lifetime Max=12
    Lifetime Min=9
    Lifetime Mode=::NewEnumerator1
    Sprite Size Mode=::NewEnumerator1
    Uniform Sprite Size Max=1.5
    Uniform Sprite Size Min=1
    UsePositionOffset=false
    Write Material Random=true
   )
   /SetVariables(Particles.SpriteRotationRate=&RandomRangeFloat(Minimum=-0.25, Maximum=0.25, Evaluation Type=::NewEnumerator1)) {isSetParameters=true, module=SetVariables_#2}
   &AddVelocity(Velocity=&RandomRangeVector(Minimum=<-10 -10 -10>, Maximum=<10 10 10>, Evaluation Type=::NewEnumerator1))
   /Niagara/Modules/Spawn/Location/SphereLocation(
    MaskSpawnGroup=false
    Module.Override Randomness=false
    Sphere Radius=400
    Surface Expansion Mode:ENiagaraExpansionMode=::NewEnumerator1
    Surface Only=true
   )
  ]
  particleUpdateScript [
   &ParticleState(Kill Particles When Lifetime Has Elapsed=false, Loop Particles Lifetime=true)
   &SpriteRotationRate(Rotation Rate=$Particles.SpriteRotationRate)
   &SampleSkeletalMesh(
    AdditionalYawPitchRoll=false
    Mesh Sampling Type:ENiagaraSkelSamplingModeFull=::NewEnumerator5
    Skeletal Mesh:NiagaraDataInterfaceSkeletalMesh=$System.Skeletal Mesh
    Socket Sampling Mode:ENiagaraSocketSamplingMode=::NewEnumerator5
   )
   /SetVariables(Particles.Velocity Speed:NiagaraFloat=&VectorLength(Vector=$Engine.Owner.Velocity)) {isSetParameters=true, module=SetVariables_#3}
   &Color(Color=&MakeLinearColorFromVectorAndFloat(
     "Vector (RGB)"=&MakeVectorFromLinearColorRGB(Color=<rgba 1 1 1 1>)
     "Float (Alpha)"=&ClampFloat(Float=&FloatFromCurve(FloatCurve=DataInterface() {propertyValues {Curve {keys [{arriveTangent=-1, time=0, value=1}, <at 1 0>]}}}))
    ))
   &ScaleColor(Scale Alpha=0.3, ScaleA=true, ScaleRGB=false)
   /Niagara/Modules/Math/FadeOverTime(
    Fade Rate:NiagaraFloat=&FloatFromCurve(FloatCurve=DataInterface() {propertyValues {Curve {keys [<at 0 0 RCIM_Cubic>,<at 1 0.25 RCIM_Cubic>]}}})
    Target Value=1
    Use Min Delta=false
   )
   &FadeOverTime(Fade Rate=0.03, Target Value=1, Use Min Delta=false) {module=FadeOverTime001}
   &PendulumSetup(
    Module.Enable Spring Driver=false
    Pendulum Length=15
    Pendulum Pivot Offset:NiagaraPosition=$Particles.SampleSkeletalMesh.SampledPosition
    Pendulum Pivot Offset Coordinate Space:ENiagaraCoordinateSpace=::Local
    Rigid Pendulum=false
   )
   &AccelerationForce(Acceleration=&Multiply_VectorByFloat(
     Vector=$Engine.Owner.Velocity
     Float=&Lerp_Float(
      A=0.1
      B=0.5
      Alpha=&FloatFromCurve(FloatCurve=DataInterface() {propertyValues {Curve {keys [
           {interpMode=RCIM_Cubic, tangentMode=RCTM_User, time=0, value=0}
           {arriveTangent=2.2885, interpMode=RCIM_Cubic, leaveTangent=2.2885, tangentMode=RCTM_User, time=1, value=1}
          ]}}})
     )
    )) {module=AccelerationForce001}
   &CurlNoiseForce(
    Noise Frequency=&RandomRangeFloat(Minimum=44, Maximum=77)
    Noise Strength=&MakeCustomFloatFromBool(
     Bool=$User.bIsFrontEnd
     True Float=&RandomRangeFloat(Minimum=2, Maximum=4)
     False Float=&Multiply_Float(A=$Particles.Velocity Speed, B=$Particles.FadeOverTime.CurrentValue)
    )
    Pan Noise=true
    Scale Curl Lookup By Engine Owner Scale=false
   ) {module=CurlNoiseForce001}
   &Drag(
    Drag=&Multiply_Float(
     A=$Particles.Velocity Speed
     B=&FloatFromCurve(
      FloatCurve=DataInterface() {propertyValues {Curve {keys [{arriveTangent=-0.1327, interpMode=RCIM_Cubic, leaveTangent=-0.1327, tangentMode=RCTM_User, time=0, value=0.2}, <at 1 0.05>]}}}
      Scale Curve=&Multiply_Float(A=1, B=2)
     )
    )
    Use Linear Drag=true
    Use Rotational Drag=false
   ) {module=Drag003}
   &Drag(
    Drag=&Multiply_Float(
     A=&OneMinusFloat(Float=$Particles.Velocity Speed)
     B=&FloatFromCurve(FloatCurve=DataInterface() {propertyValues {Curve {keys [
          {arriveTangent=5.8468, interpMode=RCIM_Cubic, leaveTangent=5.8468, tangentMode=RCTM_User, time=0, value=0}
          {arriveTangent=13.6245, interpMode=RCIM_Cubic, leaveTangent=13.6245, tangentMode=RCTM_User, time=1, value=7}
         ]}}})
    )
    Use Linear Drag=true
    Use Rotational Drag=false
   ) {module=Drag002}
   &AccelerationForce(Acceleration=&MakeCustomVectorFromBool(
     Bool=$User.bIsFrontEnd
     True Vector=<0 0 0>
     False Vector=&Multiply_VectorByFloat(
      Vector=&RandomRangeVector(Minimum=<0 0 -4>, Maximum=<0 0 -7>)
      Float=$Particles.FadeOverTime001.CurrentValue
     )
    )) {module=AccelerationForce004}
   &SolveForcesAndVelocity(Clamp Velocity=false, Limit Acceleration=false)
   &PendulumConstraint(Calculate and Output Potential Energy=true)
   &CalculateAccurateVelocity()
  ]
  rendererClasses [&NiagaraSpriteRendererProperties]
  renderers [{rendererClass=&NiagaraSpriteRendererProperties}]
 }
 Confetti001 {
  emitterUpdateScript [
   &EmitterState(
    Inactive Response=::NewEnumerator1
    Life Cycle Mode=::NewEnumerator1
    Loop Behavior=::NewEnumerator1
    Loop Duration Mode=::NewEnumerator1
    UseLoopDelay=false
   )
   &SpawnBurst_Instantaneous(Spawn Count=10, Use Spawn Probability=false, UseLoopCountLimit=false)
  ]
  particleSpawnScript [
   &InitializeParticle(
    Color Mode=::NewEnumerator0
    Lifetime Max=12
    Lifetime Min=9
    Lifetime Mode=::NewEnumerator1
    Sprite Size Mode=::NewEnumerator1
    Uniform Sprite Size Max=1.5
    Uniform Sprite Size Min=1
    UsePositionOffset=false
    Write Material Random=true
   )
   /SetVariables(Particles.SpriteRotationRate=&RandomRangeFloat(Minimum=-0.25, Maximum=0.25, Evaluation Type=::NewEnumerator1)) {isSetParameters=true, module=SetVariables_#2}
   &AddVelocity(Velocity=&RandomRangeVector(Minimum=<-10 -10 -10>, Maximum=<10 10 10>, Evaluation Type=::NewEnumerator1))
   &SphereLocation(
    MaskSpawnGroup=false
    Module.Override Randomness=false
    Sphere Radius=400
    Surface Expansion Mode:ENiagaraExpansionMode=::NewEnumerator1
    Surface Only=true
   )
  ]
  particleUpdateScript [
   &ParticleState(Kill Particles When Lifetime Has Elapsed=false, Loop Particles Lifetime=true)
   &SpriteRotationRate(Rotation Rate=$Particles.SpriteRotationRate)
   &SampleSkeletalMesh(
    AdditionalYawPitchRoll=false
    Mesh Sampling Type:ENiagaraSkelSamplingModeFull=::NewEnumerator5
    Skeletal Mesh:NiagaraDataInterfaceSkeletalMesh=$System.Skeletal Mesh
    Socket Sampling Mode:ENiagaraSocketSamplingMode=::NewEnumerator5
   )
   /SetVariables(Particles.Velocity Speed:NiagaraFloat=&VectorLength(Vector=$Engine.Owner.Velocity)) {isSetParameters=true, module=SetVariables_#3}
   &Color(Color=&MakeLinearColorFromVectorAndFloat(
     "Vector (RGB)"=&MakeVectorFromLinearColorRGB(Color=<rgba 1 1 1 1>)
     "Float (Alpha)"=&ClampFloat(Float=&FloatFromCurve(FloatCurve=DataInterface() {propertyValues {Curve {keys [{arriveTangent=-1, time=0, value=1}, <at 1 0>]}}}))
    ))
   &ScaleColor(Scale Alpha=0.3, ScaleA=true, ScaleRGB=false)
   &FadeOverTime(
    Fade Rate:NiagaraFloat=&FloatFromCurve(FloatCurve=DataInterface() {propertyValues {Curve {keys [<at 0 0 RCIM_Cubic>,<at 1 0.25 RCIM_Cubic>]}}})
    Target Value=1
    Use Min Delta=false
   )
   &FadeOverTime(Fade Rate=0.03, Target Value=1, Use Min Delta=false) {module=FadeOverTime001}
   &PendulumSetup(
    Module.Enable Spring Driver=false
    Pendulum Length=15
    Pendulum Pivot Offset:NiagaraPosition=$Particles.SampleSkeletalMesh.SampledPosition
    Pendulum Pivot Offset Coordinate Space:ENiagaraCoordinateSpace=::Local
    Rigid Pendulum=false
   )
   &AccelerationForce(Acceleration=&Multiply_VectorByFloat(
     Vector=$Engine.Owner.Velocity
     Float=&Lerp_Float(
      A=0.1
      B=0.5
      Alpha=&FloatFromCurve(FloatCurve=DataInterface() {propertyValues {Curve {keys [
           {interpMode=RCIM_Cubic, tangentMode=RCTM_User, time=0, value=0}
           {arriveTangent=2.2885, interpMode=RCIM_Cubic, leaveTangent=2.2885, tangentMode=RCTM_User, time=1, value=1}
          ]}}})
     )
    )) {module=AccelerationForce001}
   &CurlNoiseForce(
    Noise Frequency=&RandomRangeFloat(Minimum=44, Maximum=77)
    Noise Strength=&MakeCustomFloatFromBool(
     Bool=$User.bIsFrontEnd
     True Float=&RandomRangeFloat(Minimum=2, Maximum=4)
     False Float=&Multiply_Float(A=$Particles.Velocity Speed, B=$Particles.FadeOverTime.CurrentValue)
    )
    Pan Noise=true
    Scale Curl Lookup By Engine Owner Scale=false
   ) {module=CurlNoiseForce001}
   &Drag(
    Drag=&Multiply_Float(
     A=$Particles.Velocity Speed
     B=&FloatFromCurve(
      FloatCurve=DataInterface() {propertyValues {Curve {keys [{arriveTangent=-0.1327, interpMode=RCIM_Cubic, leaveTangent=-0.1327, tangentMode=RCTM_User, time=0, value=0.2}, <at 1 0.05>]}}}
      Scale Curve=&Multiply_Float(A=1, B=2)
     )
    )
    Use Linear Drag=true
    Use Rotational Drag=false
   ) {module=Drag003}
   &Drag(
    Drag=&Multiply_Float(
     A=&OneMinusFloat(Float=$Particles.Velocity Speed)
     B=&FloatFromCurve(FloatCurve=DataInterface() {propertyValues {Curve {keys [
          {arriveTangent=5.8468, interpMode=RCIM_Cubic, leaveTangent=5.8468, tangentMode=RCTM_User, time=0, value=0}
          {arriveTangent=13.6245, interpMode=RCIM_Cubic, leaveTangent=13.6245, tangentMode=RCTM_User, time=1, value=7}
         ]}}})
    )
    Use Linear Drag=true
    Use Rotational Drag=false
   ) {module=Drag002}
   &AccelerationForce(Acceleration=&MakeCustomVectorFromBool(
     Bool=$User.bIsFrontEnd
     True Vector=<0 0 0>
     False Vector=&Multiply_VectorByFloat(
      Vector=&RandomRangeVector(Minimum=<0 0 -4>, Maximum=<0 0 -7>)
      Float=$Particles.FadeOverTime001.CurrentValue
     )
    )) {module=AccelerationForce004}
   &SolveForcesAndVelocity(Clamp Velocity=false, Limit Acceleration=false)
   &PendulumConstraint(Calculate and Output Potential Energy=true)
   &CalculateAccurateVelocity()
  ]
  rendererClasses [&NiagaraSpriteRendererProperties]
  renderers [{rendererClass=&NiagaraSpriteRendererProperties}]
 }
 Snow {
  simTarget=GPUComputeSim
  emitterUpdateScript [
   &EmitterState(
    Inactive Response=::NewEnumerator1
    Life Cycle Mode=::NewEnumerator1
    Loop Behavior=::NewEnumerator1
    Loop Duration Mode=::NewEnumerator1
    UseLoopDelay=false
   )
   &SpawnBurst_Instantaneous(Spawn Count=150, Use Spawn Probability=false, UseLoopCountLimit=false)
  ]
  particleSpawnScript [
   &InitializeParticle(
    Color Mode=::NewEnumerator0
    Lifetime Max=12
    Lifetime Min=9
    Lifetime Mode=::NewEnumerator1
    Sprite Size Mode=::NewEnumerator1
    Uniform Sprite Size Max=2.75
    Uniform Sprite Size Min=1.75
    UsePositionOffset=false
    Write Material Random=true
   )
   /SetVariables(Particles.SpriteRotationRate=&RandomRangeFloat(Minimum=-0.25, Maximum=0.25, Evaluation Type=::NewEnumerator1)) {isSetParameters=true, module=SetVariables_#2}
   &AddVelocity(Velocity=&RandomRangeVector(Minimum=<-10 -10 -10>, Maximum=<10 10 10>, Evaluation Type=::NewEnumerator1))
   &SphereLocation(
    MaskSpawnGroup=false
    Module.Override Randomness=false
    Sphere Radius=400
    Surface Expansion Mode:ENiagaraExpansionMode=::NewEnumerator1
    Surface Only=true
   )
   &SubUVAnimation(
    End Frame Range Override=2
    Evaluation Type=::NewEnumerator0
    "Sprite Renderer "=DataInterface() {propertyValues {SpriteRenderer=&self:Snow_2.NiagaraSpriteRendererProperties_0}}
    SubUV Animation Mode:ENiagara_SubUVLookupModeV2=::NewEnumerator2
    UseEndFrame=true
    UseStartFrame=true
   )
  ]
  particleUpdateScript [
   &ParticleState(Kill Particles When Lifetime Has Elapsed=false, Loop Particles Lifetime=true)
   &SpriteRotationRate(Rotation Rate=$Particles.SpriteRotationRate)
   &SampleSkeletalMesh(
    AdditionalYawPitchRoll=false
    Mesh Sampling Type:ENiagaraSkelSamplingModeFull=::NewEnumerator5
    Skeletal Mesh:NiagaraDataInterfaceSkeletalMesh=$System.Skeletal Mesh
    Socket Sampling Mode:ENiagaraSocketSamplingMode=::NewEnumerator5
   )
   /SetVariables(Particles.Velocity Speed:NiagaraFloat=&VectorLength(Vector=$Engine.Owner.Velocity)) {isSetParameters=true, module=SetVariables_#3}
   &Color(Color=&MakeLinearColorFromVectorAndFloat(
     "Vector (RGB)"=&MakeVectorFromLinearColorRGB(Color=&FNTimeOfDayColor(Night Color=<rgba 0.1 0.1 0.1 1>, Morning Color=<rgba 0.75 0.75 0.75 1>, Day Color=<rgba 1 1 1 1>, Evening Color=<rgba 0.65 0.65 0.65 1>))
     "Float (Alpha)"=&RandomRangeFloat(Minimum=0.07, Maximum=0.12)
    )) {module=Color001}
   &FadeOverTime(
    Fade Rate:NiagaraFloat=&FloatFromCurve(FloatCurve=DataInterface() {propertyValues {Curve {keys [<at 0 0 RCIM_Cubic>,<at 1 0.25 RCIM_Cubic>]}}})
    Target Value=1
    Use Min Delta=false
   )
   &FadeOverTime(Fade Rate=0.03, Target Value=1, Use Min Delta=false) {module=FadeOverTime001}
   &PendulumSetup(
    Module.Enable Spring Driver=false
    Pendulum Length=15
    Pendulum Pivot Offset:NiagaraPosition=$Particles.SampleSkeletalMesh.SampledPosition
    Pendulum Pivot Offset Coordinate Space:ENiagaraCoordinateSpace=::Local
    Rigid Pendulum=false
   )
   &AccelerationForce(Acceleration=&Multiply_VectorByFloat(
     Vector=$Engine.Owner.Velocity
     Float=&Lerp_Float(
      A=0.1
      B=0.5
      Alpha=&FloatFromCurve(FloatCurve=DataInterface() {propertyValues {Curve {keys [
           {interpMode=RCIM_Cubic, tangentMode=RCTM_User, time=0, value=0}
           {arriveTangent=2.2885, interpMode=RCIM_Cubic, leaveTangent=2.2885, tangentMode=RCTM_User, time=1, value=1}
          ]}}})
     )
    )) {module=AccelerationForce001}
   &CurlNoiseForce(
    Noise Frequency=&RandomRangeFloat(Minimum=44, Maximum=77)
    Noise Strength=&MakeCustomFloatFromBool(
     Bool=$User.bIsFrontEnd
     True Float=&RandomRangeFloat(Minimum=2, Maximum=4)
     False Float=&Multiply_Float(A=$Particles.Velocity Speed, B=$Particles.FadeOverTime.CurrentValue)
    )
    Pan Noise=true
    Scale Curl Lookup By Engine Owner Scale=false
   ) {module=CurlNoiseForce001}
   &Drag(
    Drag=&Multiply_Float(
     A=$Particles.Velocity Speed
     B=&FloatFromCurve(
      FloatCurve=DataInterface() {propertyValues {Curve {keys [{arriveTangent=-0.1327, interpMode=RCIM_Cubic, leaveTangent=-0.1327, tangentMode=RCTM_User, time=0, value=0.2}, <at 1 0.05>]}}}
      Scale Curve=&Multiply_Float(A=1, B=2)
     )
    )
    Use Linear Drag=true
    Use Rotational Drag=false
   ) {module=Drag003}
   &Drag(
    Drag=&Multiply_Float(
     A=&OneMinusFloat(Float=$Particles.Velocity Speed)
     B=&FloatFromCurve(FloatCurve=DataInterface() {propertyValues {Curve {keys [
          {arriveTangent=5.8468, interpMode=RCIM_Cubic, leaveTangent=5.8468, tangentMode=RCTM_User, time=0, value=0}
          {arriveTangent=13.6245, interpMode=RCIM_Cubic, leaveTangent=13.6245, tangentMode=RCTM_User, time=1, value=7}
         ]}}})
    )
    Use Linear Drag=true
    Use Rotational Drag=false
   ) {module=Drag002}
   &AccelerationForce(Acceleration=&MakeCustomVectorFromBool(
     Bool=$User.bIsFrontEnd
     True Vector=<0 0 0>
     False Vector=&Multiply_VectorByFloat(
      Vector=&RandomRangeVector(Minimum=<0 0 -4>, Maximum=<0 0 -7>)
      Float=$Particles.FadeOverTime001.CurrentValue
     )
    )) {module=AccelerationForce004}
   &SolveForcesAndVelocity(Clamp Velocity=false, Limit Acceleration=false)
   &PendulumConstraint(Calculate and Output Potential Energy=true)
   &CalculateAccurateVelocity()
  ]
  rendererClasses [&NiagaraSpriteRendererProperties]
  renderers [{rendererClass=&NiagaraSpriteRendererProperties}]
 }
 Snow001 {
  emitterUpdateScript [
   &EmitterState(
    Inactive Response=::NewEnumerator1
    Life Cycle Mode=::NewEnumerator1
    Loop Behavior=::NewEnumerator1
    Loop Duration Mode=::NewEnumerator1
    UseLoopDelay=false
   )
   &SpawnBurst_Instantaneous(Spawn Count=20, Use Spawn Probability=false, UseLoopCountLimit=false)
  ]
  particleSpawnScript [
   &InitializeParticle(
    Color Mode=::NewEnumerator0
    Lifetime Max=12
    Lifetime Min=9
    Lifetime Mode=::NewEnumerator1
    Sprite Size Mode=::NewEnumerator1
    Uniform Sprite Size Max=2.75
    Uniform Sprite Size Min=1.75
    UsePositionOffset=false
    Write Material Random=true
   )
   /SetVariables(Particles.SpriteRotationRate=&RandomRangeFloat(Minimum=-0.25, Maximum=0.25, Evaluation Type=::NewEnumerator1)) {isSetParameters=true, module=SetVariables_#2}
   &AddVelocity(Velocity=&RandomRangeVector(Minimum=<-10 -10 -10>, Maximum=<10 10 10>, Evaluation Type=::NewEnumerator1))
   &SphereLocation(
    MaskSpawnGroup=false
    Module.Override Randomness=false
    Sphere Radius=400
    Surface Expansion Mode:ENiagaraExpansionMode=::NewEnumerator1
    Surface Only=true
   )
   &SubUVAnimation(
    End Frame Range Override=2
    Evaluation Type=::NewEnumerator0
    "Sprite Renderer "=DataInterface() {propertyValues {SpriteRenderer=&self:Snow001_1.NiagaraSpriteRendererProperties_0}}
    SubUV Animation Mode:ENiagara_SubUVLookupModeV2=::NewEnumerator2
    UseEndFrame=true
    UseStartFrame=true
   )
  ]
  particleUpdateScript [
   &ParticleState(Kill Particles When Lifetime Has Elapsed=false, Loop Particles Lifetime=true)
   &SpriteRotationRate(Rotation Rate=$Particles.SpriteRotationRate)
   &SampleSkeletalMesh(
    AdditionalYawPitchRoll=false
    Mesh Sampling Type:ENiagaraSkelSamplingModeFull=::NewEnumerator5
    Skeletal Mesh:NiagaraDataInterfaceSkeletalMesh=$System.Skeletal Mesh
    Socket Sampling Mode:ENiagaraSocketSamplingMode=::NewEnumerator5
   )
   /SetVariables(Particles.Velocity Speed:NiagaraFloat=&VectorLength(Vector=$Engine.Owner.Velocity)) {isSetParameters=true, module=SetVariables_#3}
   &Color(Color=&MakeLinearColorFromVectorAndFloat(
     "Vector (RGB)"=&MakeVectorFromLinearColorRGB(Color=&FNTimeOfDayColor(Night Color=<rgba 0.1 0.1 0.1 1>, Morning Color=<rgba 0.75 0.75 0.75 1>, Day Color=<rgba 1 1 1 1>, Evening Color=<rgba 0.65 0.65 0.65 1>))
     "Float (Alpha)"=&RandomRangeFloat(Minimum=0.07, Maximum=0.12)
    )) {module=Color001}
   &FadeOverTime(
    Fade Rate:NiagaraFloat=&FloatFromCurve(FloatCurve=DataInterface() {propertyValues {Curve {keys [<at 0 0 RCIM_Cubic>,<at 1 0.25 RCIM_Cubic>]}}})
    Target Value=1
    Use Min Delta=false
   )
   &FadeOverTime(Fade Rate=0.03, Target Value=1, Use Min Delta=false) {module=FadeOverTime001}
   &PendulumSetup(
    Module.Enable Spring Driver=false
    Pendulum Length=15
    Pendulum Pivot Offset:NiagaraPosition=$Particles.SampleSkeletalMesh.SampledPosition
    Pendulum Pivot Offset Coordinate Space:ENiagaraCoordinateSpace=::Local
    Rigid Pendulum=false
   )
   &AccelerationForce(Acceleration=&Multiply_VectorByFloat(
     Vector=$Engine.Owner.Velocity
     Float=&Lerp_Float(
      A=0.1
      B=0.5
      Alpha=&FloatFromCurve(FloatCurve=DataInterface() {propertyValues {Curve {keys [
           {interpMode=RCIM_Cubic, tangentMode=RCTM_User, time=0, value=0}
           {arriveTangent=2.2885, interpMode=RCIM_Cubic, leaveTangent=2.2885, tangentMode=RCTM_User, time=1, value=1}
          ]}}})
     )
    )) {module=AccelerationForce001}
   &CurlNoiseForce(
    Noise Frequency=&RandomRangeFloat(Minimum=44, Maximum=77)
    Noise Strength=&MakeCustomFloatFromBool(
     Bool=$User.bIsFrontEnd
     True Float=&RandomRangeFloat(Minimum=2, Maximum=4)
     False Float=&Multiply_Float(A=$Particles.Velocity Speed, B=$Particles.FadeOverTime.CurrentValue)
    )
    Pan Noise=true
    Scale Curl Lookup By Engine Owner Scale=false
   ) {module=CurlNoiseForce001}
   &Drag(
    Drag=&Multiply_Float(
     A=$Particles.Velocity Speed
     B=&FloatFromCurve(
      FloatCurve=DataInterface() {propertyValues {Curve {keys [{arriveTangent=-0.1327, interpMode=RCIM_Cubic, leaveTangent=-0.1327, tangentMode=RCTM_User, time=0, value=0.2}, <at 1 0.05>]}}}
      Scale Curve=&Multiply_Float(A=1, B=2)
     )
    )
    Use Linear Drag=true
    Use Rotational Drag=false
   ) {module=Drag003}
   &Drag(
    Drag=&Multiply_Float(
     A=&OneMinusFloat(Float=$Particles.Velocity Speed)
     B=&FloatFromCurve(FloatCurve=DataInterface() {propertyValues {Curve {keys [
          {arriveTangent=5.8468, interpMode=RCIM_Cubic, leaveTangent=5.8468, tangentMode=RCTM_User, time=0, value=0}
          {arriveTangent=13.6245, interpMode=RCIM_Cubic, leaveTangent=13.6245, tangentMode=RCTM_User, time=1, value=7}
         ]}}})
    )
    Use Linear Drag=true
    Use Rotational Drag=false
   ) {module=Drag002}
   &AccelerationForce(Acceleration=&MakeCustomVectorFromBool(
     Bool=$User.bIsFrontEnd
     True Vector=<0 0 0>
     False Vector=&Multiply_VectorByFloat(
      Vector=&RandomRangeVector(Minimum=<0 0 -4>, Maximum=<0 0 -7>)
      Float=$Particles.FadeOverTime001.CurrentValue
     )
    )) {module=AccelerationForce004}
   &SolveForcesAndVelocity(Clamp Velocity=false, Limit Acceleration=false)
   &PendulumConstraint(Calculate and Output Potential Energy=true)
   &CalculateAccurateVelocity()
  ]
  rendererClasses [&NiagaraSpriteRendererProperties]
  renderers [{rendererClass=&NiagaraSpriteRendererProperties}]
 }
 SnowGlow {
  emitterUpdateScript [
   &EmitterState(
    Inactive Response=::NewEnumerator1
    Life Cycle Mode=::NewEnumerator1
    Loop Behavior=::NewEnumerator1
    Loop Duration Mode=::NewEnumerator1
    UseLoopDelay=false
   )
   &SpawnBurst_Instantaneous(Use Spawn Probability=false, UseLoopCountLimit=false)
  ]
  particleSpawnScript [
   &InitializeParticle(
    Color=<rgba 0.2385 0.3219 0.5 1>
    Lifetime=2
    Sprite Size Mode=::NewEnumerator3
    Uniform Sprite Size=40
    UsePositionOffset=false
    Write Material Random=true
   )
   &CameraOffset(Camera Offset Amount=4)
   /SetVariables(Particles.Color Low:LinearColor=<rgba 0.2385 0.3219 0.5 0.7>) {isSetParameters=true}
   /SetVariables(Particles.LowCameraOffset=15) {isSetParameters=true}
  ]
  particleUpdateScript [
   &ParticleState(Kill Particles When Lifetime Has Elapsed=false)
   &ScaleSpriteSize(Uniform Curve Sprite Scale=DataInterface() {propertyValues {Curve {keys [{tangentMode=RCTM_User, time=0, value=0}, {tangentMode=RCTM_User, time=0.18, value=1}]}}})
   &SkeletalMeshLocation(MaskSpawnGroup=false, Mesh Sampling Type=::NewEnumerator5, Skeletal Mesh=$System.Skeletal Mesh, Socket Sampling Mode=::NewEnumerator5)
   &DynamicMaterialParameters(
    Index 0 Param 1=1.7
    Index 0 Param 2=8
    Module.Index0Param0Write=true
    Module.Index0Param1Write=true
    Module.Index0Param2Write=true
    Module.Index0Param3Write=true
   )
   &ScaleColor(
    Scale RGB=&Multiply_VectorByFloat(
     Vector=<1 1 1>
     Float=&FN_Select_Float_from_Time_Of_Day(Morning=0.5, Day=1, Evening=0.5, Night=0.5)
    )
    ScaleA=true
    ScaleRGB=true
   )
  ]
  rendererClasses [&NiagaraSpriteRendererProperties]
  renderers [
   {rendererClass=&NiagaraSpriteRendererProperties}
   {rendererClass=&NiagaraSpriteRendererProperties, rendererIndex=1}
  ]
 }
}
summary {
 systemName=NS_Winterfest1_Pickaxe_Idle
 emitters [
  {emitterName=Snow, simTarget=GPUComputeSim, rendererClasses [&NiagaraSpriteRendererProperties]}
  {emitterName=Confetti, simTarget=GPUComputeSim, rendererClasses [&NiagaraSpriteRendererProperties]}
  {emitterName=SnowGlow, rendererClasses [&NiagaraSpriteRendererProperties]}
  {emitterName=Snow001, rendererClasses [&NiagaraSpriteRendererProperties]}
  {emitterName=Confetti001, rendererClasses [&NiagaraSpriteRendererProperties]}
 ]
 userVariables [
  User.bIsFrontEnd {type=NiagaraBool}
  User.bIsFrontEndPreview {type=NiagaraBool}
 ]
}
