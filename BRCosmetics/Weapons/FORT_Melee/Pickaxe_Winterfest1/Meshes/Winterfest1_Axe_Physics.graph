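# Physics asset for the Winterfest1 pickaxe mesh, setting up secondary motion for its hat and left/right ribbons.
# Small capsules on the base bones act as roots, and slightly larger capsules cover the dynamic bones.
# Swing- and twist-limited constraints tie each dynamic bone to its parent, with the b ribbons forming two-link chains per side.
# The a ribbons have tighter swing than the hat and b ribbons, while the hat allows the most twist.
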PhysicsAsset /BRCosmetics/Weapons/FORT_Melee/Pickaxe_Winterfest1/Meshes/Winterfest1_Axe_Physics
bodies {
 base_hat [{shapeType=Capsule, length=0.505, radius=0.505, shapeName=base_hat_capsule, rotation=<rot 0 -90 0>}]
 base_ribbon_a_l [{shapeType=Capsule, length=0.505, radius=0.505, shapeName=base_ribbon_a_l_capsule, rotation=<rot 0 -90 0>}]
 base_ribbon_a_r [{shapeType=Capsule, length=0.505, radius=0.505, shapeName=base_ribbon_a_r_capsule, rotation=<rot 0 -90 0>}]
 base_ribbon_b_l [{shapeType=Capsule, length=0.505, radius=0.505, shapeName=base_ribbon_b_l_capsule, rotation=<rot 0 -90 0>}]
 base_ribbon_b_r [{shapeType=Capsule, length=0.505, radius=0.505, shapeName=base_ribbon_b_r_capsule, rotation=<rot 0 -90 0>}]
 dyn_hat [{shapeType=Capsule, length=0.4389, radius=1, shapeName=dyn_hat_capsule, center=<1 0 0>, rotation=<rot 90 0 0>}]
 dyn_ribbon_a_l [{shapeType=Capsule, length=2, radius=1, shapeName=dyn_ribbon_a_l_capsule, center=<2 0 0>, rotation=<rot 90 0 0>}]
 dyn_ribbon_a_r [{shapeType=Capsule, length=2, radius=1, shapeName=dyn_ribbon_a_r_capsule, center=<2 0 0>, rotation=<rot 90 0 0>}]
 dyn_ribbon_b_00_l [{shapeType=Capsule, length=2, radius=1, shapeName=dyn_ribbon_b_00_l_capsule, center=<2 0 0>, rotation=<rot 90 0 0>}]
 dyn_ribbon_b_00_r [{shapeType=Capsule, length=2, radius=1, shapeName=dyn_ribbon_b_00_r_capsule, center=<2 0 0>, rotation=<rot 90 0 0>}]
 dyn_ribbon_b_01_l [{shapeType=Capsule, length=2, radius=1, shapeName=dyn_ribbon_b_01_l_capsule, center=<2 0 0>, rotation=<rot 90 0 0>}]
 dyn_ribbon_b_01_r [{shapeType=Capsule, length=2, radius=1, shapeName=dyn_ribbon_b_01_r_capsule, center=<2 0 0>, rotation=<rot 90 0 0>}]
}
constraints [
 {bone1Name=dyn_ribbon_b_01_r, bone2Name=dyn_ribbon_b_00_r, swing1LimitDegrees=25, swing2LimitDegrees=25, twistLimitDegrees=5}
 {bone1Name=dyn_hat, bone2Name=base_hat, swing1LimitDegrees=25, swing2LimitDegrees=25, twistLimitDegrees=10}
 {bone1Name=dyn_ribbon_a_l, bone2Name=base_ribbon_a_l, swing1LimitDegrees=15, swing2LimitDegrees=15, twistLimitDegrees=5}
 {bone1Name=dyn_ribbon_a_r, bone2Name=base_ribbon_a_r, swing1LimitDegrees=15, swing2LimitDegrees=15, twistLimitDegrees=5}
 {bone1Name=dyn_ribbon_b_00_l, bone2Name=base_ribbon_b_l, swing1LimitDegrees=25, swing2LimitDegrees=25, twistLimitDegrees=5}
 {bone1Name=dyn_ribbon_b_01_l, bone2Name=dyn_ribbon_b_00_l, swing1LimitDegrees=25, swing2LimitDegrees=25, twistLimitDegrees=5}
 {bone1Name=dyn_ribbon_b_00_r, bone2Name=base_ribbon_b_r, swing1LimitDegrees=25, swing2LimitDegrees=25, twistLimitDegrees=5}
]
